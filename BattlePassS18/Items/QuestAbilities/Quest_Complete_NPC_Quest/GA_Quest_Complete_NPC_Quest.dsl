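; Battle Pass S18 quest ability meant to credit completion of NPC quests, apparently currently broken.
; On activation it only calls the parent and prints debug text, plus a red warning that the OnTransientQuestRemoved event it relied on no longer exists.
; The TransientQuestRemoved handler, no longer bound here, checks the quest is completed, then sends a tagged stat event to the owner's quest manager and calls an empty Cleanup event.
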(event Ability|EventActivateAbility
  (|CallParentFunction)
  (Development|PrintString "@@@ ActivateAbility [GA_Quest_Complete_NPC_Quest]: CALLED!" false (Variables|Debug|GetbDebugAbility))
  (Development|PrintString "OnTransientQuestRemoved event no longer Exists! If trying to revive this component, you'll need to bind to the right OnQuest removed event depending on your needs" false true "(R=1.000000,G=0.000000,B=0.000000,A=1.000000)" 2e+01))

(event Ability|EventFailedToActivatePassiveAbility
  (params (ActorInfo :type Gameplay%20Ability%20Actor%20Info%20Structure)))

(event Custom|TransientQuestRemoved
  (params (TransientQuestItem :type Fort%20Quest%20Item%20Object%20Reference))
  (bind _returnvalue (Quest|HasCompletedQuest TransientQuestItem :!class /Script/FortniteGame.FortQuestItem))
  (if _returnvalue
    (Quest|SendComplexCustomStatEvent (Variables|GAVars|Data|GetOwningPlayerQuestManager) 0 (Variables|Default|GetQuestStatEventTags) "(GameplayTags=)" 1)
    (Development|PrintString (Utilities|String|Append "@@@@@@ ActivateAbility [GA_Quest_Complete_NPC_Quest]: Transient quest removed caught - Quest item = " (Utilities|GetDisplayName TransientQuestItem) "; Has Completed = " (Utilities|String|ToString(Boolean) _returnvalue)) false (Variables|Debug|GetbDebugAbility))
    (CallFunction|Cleanup)))

(event Custom|Cleanup)
